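# Sound cue that plays the arcade cabinet success sound for a Season 24 Battle Pass quest, judging by its name and path.
# A single wave player feeds a Random node with one entry, then a root Modulator that adds very slight downward pitch variation at constant volume.
# The cue is routed to the Fort_Quests sound class.
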SoundCue /BattlePassS24/Audio/SFX/S24_Pt01_Q01_ArcadeCabinet_Success_Cue
root=Modulator_0
nodes [
 {id=Modulator_0, children [Random_0], properties {pitchMax=1, pitchMin=0.99, volumeMax=1, volumeMin=1}}
 {id=Random_0, children [WavePlayer_1], properties {weights [1]}}
 {id=WavePlayer_1, properties {soundWaveAssetPtr=&/S24_Pt01_Q01_ArcadeCabinet_Success}}
]
properties {maxDistance=2097152, soundClassObject=&Fort_Quests, volumeMultiplier=1}
