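# Sound cue for a target hit effect in a Season 24 Battle Pass quest (Part 1, Quest 3), routed to the quests sound class.
# A Random node picks among six equally weighted hit wave variations and feeds a Modulator at the root.
# The cue plays at slightly reduced volume with the quest's shared attenuation asset limiting its audible range.
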SoundCue /BattlePassS24/Audio/SFX/S24_Pt01_Q03_Target_Hit_Cue
root=Modulator_0
nodes [
 {id=Modulator_0, children [Random_0]}
 {id=Random_0, children [WavePlayer_0,WavePlayer_1,WavePlayer_2,WavePlayer_3,WavePlayer_4,WavePlayer_5], properties {weights [1,1,1,1,1,1]}}
 {id=WavePlayer_0, properties {soundWaveAssetPtr=&/S24_Pt01_Q03_Target_Hit_01}}
 {id=WavePlayer_1, properties {soundWaveAssetPtr=&/S24_Pt01_Q03_Target_Hit_02}}
 {id=WavePlayer_2, properties {soundWaveAssetPtr=&/S24_Pt01_Q03_Target_Hit_03}}
 {id=WavePlayer_3, properties {soundWaveAssetPtr=&/S24_Pt01_Q03_Target_Hit_04}}
 {id=WavePlayer_4, properties {soundWaveAssetPtr=&/S24_Pt01_Q03_Target_Hit_05}}
 {id=WavePlayer_5, properties {soundWaveAssetPtr=&/S24_Pt01_Q03_Target_Hit_06}}
]
properties {
 attenuationSettings=&/S24_Pt01_Q03_Atten
 maxDistance=2850
 soundClassObject=&Fort_Quests
 volumeMultiplier=0.8
}
