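# Sound cue for the target success sound of a Season 24 Battle Pass quest (Part 01, Q03).
# A Random node picks one of six success wave variations with equal weight.
# A Modulator at the root processes the chosen sound, and the cue plays through the Fort_Quests sound class at reduced volume with a very large max distance.
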SoundCue /BattlePassS24/Audio/SFX/S24_Pt01_Q03_Target_Success_Cue
root=Modulator_0
nodes [
 {id=Modulator_0, children [Random_1]}
 {
  id=Random_1
  children [WavePlayer_6, WavePlayer_7, WavePlayer_8, WavePlayer_9, WavePlayer_10, WavePlayer_11]
  properties {weights [1,1,1,1,1,1]}
 }
 {id=WavePlayer_10, properties {soundWaveAssetPtr=&/S24_Pt01_Q03_Target_Success_05}}
 {id=WavePlayer_11, properties {soundWaveAssetPtr=&/S24_Pt01_Q03_Target_Success_06}}
 {id=WavePlayer_6, properties {soundWaveAssetPtr=&/S24_Pt01_Q03_Target_Success_01}}
 {id=WavePlayer_7, properties {soundWaveAssetPtr=&/S24_Pt01_Q03_Target_Success_02}}
 {id=WavePlayer_8, properties {soundWaveAssetPtr=&/S24_Pt01_Q03_Target_Success_03}}
 {id=WavePlayer_9, properties {soundWaveAssetPtr=&/S24_Pt01_Q03_Target_Success_04}}
]
properties {maxDistance=2097152, soundClassObject=&Fort_Quests, volumeMultiplier=0.7}
