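# Sound cue for the activation effect of the Signal Override quest in the Season 24 Battle Pass content.
# A Random node picks one of three equally weighted activation wave variants.
# A root Modulator adds very slight pitch and volume variation to each play.
# Output routes through the Fort_Quests sound class.
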SoundCue /BattlePassS24/Audio/SFX/S24_Pt04_Q03_SignalOverride_Activate_Cue
root=Modulator_0
nodes [
 {id=Modulator_0, children [Random_1], properties {pitchMax=1.02, pitchMin=0.998, volumeMax=1, volumeMin=0.98}}
 {id=Random_1, children [WavePlayer_0,WavePlayer_7,WavePlayer_8], properties {weights [1,1,1]}}
 {id=WavePlayer_0, properties {soundWaveAssetPtr=&/S24_Pt4_Q03_SingalOverride_Activate_03}}
 {id=WavePlayer_7, properties {soundWaveAssetPtr=&/S24_Pt4_Q03_SingalOverride_Activate_01}}
 {id=WavePlayer_8, properties {soundWaveAssetPtr=&/S24_Pt4_Q03_SingalOverride_Activate_02}}
]
properties {maxDistance=2097152, soundClassObject=&Fort_Quests, volumeMultiplier=1}
