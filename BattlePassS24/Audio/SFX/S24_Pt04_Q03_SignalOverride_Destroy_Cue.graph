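# Sound cue for the destruction of the Signal Override object in a Season 24 Battle Pass quest (Part 4, Q03).
# A random node picks one of six equally weighted break wave variations.
# A root modulator adds slight pitch variation at constant volume.
# The cue plays through the Fort_Quests sound class with a very large max distance.
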SoundCue /BattlePassS24/Audio/SFX/S24_Pt04_Q03_SignalOverride_Destroy_Cue
root=Modulator_0
nodes [
 {id=Modulator_0, children [Random_1], properties {pitchMax=1.02, pitchMin=0.98, volumeMax=1, volumeMin=1}}
 {id=Random_1, children [WavePlayer_1,WavePlayer_2,WavePlayer_3,WavePlayer_4,WavePlayer_5,WavePlayer_6], properties {weights [1,1,1,1,1,1]}}
 {id=WavePlayer_1, properties {soundWaveAssetPtr=&/S24_Pt4_Q03_SingalOverride_Break_01}}
 {id=WavePlayer_2, properties {soundWaveAssetPtr=&/S24_Pt4_Q03_SingalOverride_Break_02}}
 {id=WavePlayer_3, properties {soundWaveAssetPtr=&/S24_Pt4_Q03_SingalOverride_Break_03}}
 {id=WavePlayer_4, properties {soundWaveAssetPtr=&/S24_Pt4_Q03_SingalOverride_Break_04}}
 {id=WavePlayer_5, properties {soundWaveAssetPtr=&/S24_Pt4_Q03_SingalOverride_Break_05}}
 {id=WavePlayer_6, properties {soundWaveAssetPtr=&/S24_Pt4_Q03_SingalOverride_Break_06}}
]
properties {maxDistance=2097152, soundClassObject=&Fort_Quests, volumeMultiplier=1}
